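# Post-copy AnimBP for the PointyTemper medium male body cosmetic, adding secondary motion to the pose copied from the parent mesh.
# A weighted Tail_parent constraint feeds a Control Rig given NotEmote, amulet bone offsets (one DBNO-only), then chained AnimDynamics for shorts, sleeves, amulet, laces and left bracer.
# On customization load, long backblings hide the tail part.
# The weapon-type switch drives nothing, and two bracer dynamics nodes sit outside the output chain.
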AnimBlueprint /BRCosmetics/Characters/Player/Male/Medium/Bodies/M_MED_PointyTemper/Meshes/M_MED_PointyTemper_AnimBP
target_skeleton=&/M_MED_PointyTemper_Skeleton
graphs [
 AnimGraph {nodes [
   n1 = Anim Dynamics - Bone: Dyn_bracer_L_down(Alpha=1, ExternalForce="0,0,0") {
    angularDampingOverride=0.9999
    angularSpringConstant=500
    bLinearSpring=true
    bOverrideLinearDamping=true
    bUseGravityOverride=true
    linearDampingOverride=0.9999
    linearSpringConstant=50
    simulationSpace=BoneRelative
    alphaBoolBlend {blendOption=Linear}
    boundBone {boneName=Dyn_bracer_L_down}
    componentAppliedLinearAccClamp=<5000 5000 5000>
    componentLinearAccScale=<1 1 1>
    componentLinearVelScale=<1 1 1>
    gravityOverride {z=-980}
    physicsBodyDefinitions [{
      boundBone {boneName=Dyn_bracer_L_down}
      boxExtents=<10 10 10>
      constraintSetup {angularLimitsMax=<0 0 0>, angularLimitsMin=<0 0 0>, angularTarget=<1 0 0>, linearAxesMin=<0 -2 0>}
      localJointOffset=<1 0 0>
     }]
    relativeSpaceBone {boneName=lowerarm_l}
   }
   Anim Dynamics - Bone: Dyn_bracer_L_up(ComponentPose=@n1, Alpha=1, ExternalForce="0,0,0") {
    angularDampingOverride=0.9999
    angularSpringConstant=500
    bLinearSpring=true
    bOverrideLinearDamping=true
    bUseGravityOverride=true
    linearDampingOverride=0.9999
    linearSpringConstant=50
    simulationSpace=BoneRelative
    alphaBoolBlend {blendOption=Linear}
    boundBone {boneName=Dyn_bracer_L_up}
    componentAppliedLinearAccClamp=<5000 5000 5000>
    componentLinearAccScale=<1 1 1>
    componentLinearVelScale=<1 1 1>
    gravityOverride {z=-980}
    physicsBodyDefinitions [{
      boundBone {boneName=Dyn_bracer_L_up}
      boxExtents=<10 10 10>
      constraintSetup {angularLimitsMax=<0 0 0>, angularLimitsMin=<0 0 0>, angularTarget=<1 0 0>, linearAxesMax=<0 2 0>}
      localJointOffset=<1 0 0>
     }]
    relativeSpaceBone {boneName=lowerarm_l}
   }
   n2 = "Constraint - Tail_parent (PPPPPPPP)"!(
    ComponentPose=Local To Component(LocalPose=Copy Pose From Mesh() {bUseAttachedParent=true})
    bAlphaBoolEnabled:bool=True
    ConstraintWeights_0=0.6553
    ConstraintWeights_1:float=0.0009
    ConstraintWeights_2:float=0.0012
    ConstraintWeights_3:float=0.0005
    ConstraintWeights_4:float=0.0411
    ConstraintWeights_5:float=0.1609
    ConstraintWeights_6:float=0.0404
    ConstraintWeights_7:float=0.0996
   ) -> (Pose) {
    alphaInputType=Bool
    alphaBoolBlend {blendOption=Linear}
    boneToModify {boneName=Tail_parent}
    constraintSetup [
     {targetBone {boneName=pelvis}}
     {targetBone {boneName=spine_01}}
     {targetBone {boneName=spine_02}}
     {targetBone {boneName=spine_03}}
     {targetBone {boneName=thigh_twist_01_l}}
     {targetBone {boneName=deform_glute_l}}
     {targetBone {boneName=thigh_twist_01_r}}
     {targetBone {boneName=deform_glute_r}}
    ]
    constraintWeights [1,1,1,1,1,1,1,1]
   }
   n3 = Control Rig!(Source=Component To Local(ComponentPose=@n2), Alpha, NotEmote:bool=Get NotEmote() -> (NotEmote:bool)) -> (Pose) {
    alphaBoolBlend {blendOption=Linear}
    controlRigAssetReference {blueprintRigClass=&/M_MED_PointyTemper_CtrlRig.M_MED_PointyTemper_CtrlRig_C}
   }
   n4 = "Transform (Modify) Bone - Bone: Dyn_amulet_03"!(
    ComponentPose=Local To Component(LocalPose=@n3)
    bAlphaBoolEnabled:bool=Get bIsDBNO()
    Translation="0,0,0"
    Rotation="0,0,-20"
    Scale="1,1,1"
   ) -> (Pose) {
    alphaInputType=Bool
    rotationMode=BMM_Additive
    scaleMode=BMM_Additive
    translationMode=BMM_Additive
    alphaBoolBlend {blendOption=Linear}
    boneToModify {boneName=Dyn_amulet_03}
    rotation {roll=-20}
   }
   n5 = "Transform (Modify) Bone - Bone: Dyn_amulet_01"!(ComponentPose=@n4, bAlphaBoolEnabled:bool=True, Translation="0,0,0", Rotation="0,0,-1.5", Scale="1,1,1") -> (Pose) {
    alphaInputType=Bool
    rotationMode=BMM_Additive
    scaleMode=BMM_Additive
    translationMode=BMM_Additive
    alphaBoolBlend {blendOption=Linear}
    boneToModify {boneName=Dyn_amulet_01}
    rotation {roll=-1.5}
   }
   n6 = Anim Dynamics - Bone: Dyn_shorts_out_L(ComponentPose=@n5, Alpha=1, ExternalForce="0,0,0") {
    angularDampingOverride=0.9999
    angularSpringConstant=500
    bAngularSpring=true
    bLinearSpring=true
    bOverrideAngularDamping=true
    bOverrideLinearDamping=true
    bUseGravityOverride=true
    linearDampingOverride=0.9999
    linearSpringConstant=50
    simulationSpace=RootRelative
    alphaBoolBlend {blendOption=Linear}
    boundBone {boneName=Dyn_shorts_out_L}
    componentAppliedLinearAccClamp=<5000 5000 5000>
    componentLinearAccScale=<0.5 0.5 0.5>
    componentLinearVelScale=<0.5 0.5 0.5>
    gravityOverride {z=-980}
    physicsBodyDefinitions [{
      boundBone {boneName=Dyn_shorts_out_L}
      boxExtents=<10 10 10>
      constraintSetup {angularLimitsMax=<10 4 10>, angularLimitsMin=<-10 -7 -10>, angularTarget=<1 0 0>, linearAxesMax=<0 1 1>, linearAxesMin=<0 -1 -1>}
      localJointOffset=<0.7 0 0>
     }]
   }
   n7 = Anim Dynamics - Bone: Dyn_shorts_bk_L(ComponentPose=@n6, Alpha=1, ExternalForce="0,0,0") {
    angularDampingOverride=0.9999
    angularSpringConstant=500
    bAngularSpring=true
    bLinearSpring=true
    bOverrideAngularDamping=true
    bOverrideLinearDamping=true
    bUseGravityOverride=true
    linearDampingOverride=0.9999
    linearSpringConstant=50
    simulationSpace=RootRelative
    alphaBoolBlend {blendOption=Linear}
    boundBone {boneName=Dyn_shorts_bk_L}
    componentAppliedLinearAccClamp=<5000 5000 5000>
    componentLinearAccScale=<0.5 0.5 0.5>
    componentLinearVelScale=<0.5 0.5 0.5>
    gravityOverride {z=-980}
    physicsBodyDefinitions [{
      boundBone {boneName=Dyn_shorts_bk_L}
      boxExtents=<10 10 10>
      constraintSetup {angularLimitsMax=<10 4 10>, angularLimitsMin=<-10 -7 -10>, angularTarget=<1 0 0>, linearAxesMax=<0 1 1>, linearAxesMin=<0 -1 -1>}
      localJointOffset=<0.7 0 0>
     }]
   }
   n8 = Anim Dynamics - Bone: Dyn_shorts_in_L(ComponentPose=@n7, Alpha=1, ExternalForce="0,0,0") {
    angularDampingOverride=0.9999
    angularSpringConstant=500
    bAngularSpring=true
    bLinearSpring=true
    bOverrideAngularDamping=true
    bOverrideLinearDamping=true
    bUseGravityOverride=true
    linearDampingOverride=0.9999
    linearSpringConstant=50
    simulationSpace=RootRelative
    alphaBoolBlend {blendOption=Linear}
    boundBone {boneName=Dyn_shorts_in_L}
    componentAppliedLinearAccClamp=<5000 5000 5000>
    componentLinearAccScale=<0.5 0.5 0.5>
    componentLinearVelScale=<0.5 0.5 0.5>
    gravityOverride {z=-980}
    physicsBodyDefinitions [{
      boundBone {boneName=Dyn_shorts_in_L}
      boxExtents=<10 10 10>
      constraintSetup {angularLimitsMax=<10 4 10>, angularLimitsMin=<-10 -7 -10>, angularTarget=<1 0 0>, linearAxesMax=<0 1 1>, linearAxesMin=<0 -1 -1>}
      localJointOffset=<0.7 0 0>
     }]
   }
   n9 = Anim Dynamics - Bone: Dyn_shorts_fr_L(ComponentPose=@n8, Alpha=1, ExternalForce="0,0,0") {
    angularDampingOverride=0.9999
    angularSpringConstant=500
    bAngularSpring=true
    bLinearSpring=true
    bOverrideAngularDamping=true
    bOverrideLinearDamping=true
    bUseGravityOverride=true
    linearDampingOverride=0.9999
    linearSpringConstant=50
    simulationSpace=RootRelative
    alphaBoolBlend {blendOption=Linear}
    boundBone {boneName=Dyn_shorts_fr_L}
    componentAppliedLinearAccClamp=<5000 5000 5000>
    componentLinearAccScale=<0.5 0.5 0.5>
    componentLinearVelScale=<0.5 0.5 0.5>
    gravityOverride {z=-980}
    physicsBodyDefinitions [{
      boundBone {boneName=Dyn_shorts_fr_L}
      boxExtents=<10 10 10>
      constraintSetup {angularLimitsMax=<10 0 10>, angularLimitsMin=<-10 -7 -10>, angularTarget=<1 0 0>, linearAxesMax=<0 1 1>, linearAxesMin=<0 -1 -1>}
      localJointOffset=<0.7 0 0>
     }]
   }
   n10 = Anim Dynamics - Bone: Dyn_shorts_in_R(ComponentPose=@n9, Alpha=1, ExternalForce="0,0,0") {
    angularDampingOverride=0.9999
    angularSpringConstant=500
    bAngularSpring=true
    bLinearSpring=true
    bOverrideAngularDamping=true
    bOverrideLinearDamping=true
    bUseGravityOverride=true
    linearDampingOverride=0.9999
    linearSpringConstant=50
    simulationSpace=RootRelative
    alphaBoolBlend {blendOption=Linear}
    boundBone {boneName=Dyn_shorts_in_R}
    componentAppliedLinearAccClamp=<5000 5000 5000>
    componentLinearAccScale=<0.5 0.5 0.5>
    componentLinearVelScale=<0.5 0.5 0.5>
    gravityOverride {z=-980}
    physicsBodyDefinitions [{
      boundBone {boneName=Dyn_shorts_in_R}
      boxExtents=<10 10 10>
      constraintSetup {angularLimitsMax=<10 4 10>, angularLimitsMin=<-10 -7 -10>, angularTarget=<1 0 0>, linearAxesMax=<0 1 1>, linearAxesMin=<0 -1 -1>}
      localJointOffset=<0.7 0 0>
     }]
   }
   n11 = Anim Dynamics - Bone: Dyn_shorts_fr_R(ComponentPose=@n10, Alpha=1, ExternalForce="0,0,0") {
    angularDampingOverride=0.9999
    angularSpringConstant=500
    bAngularSpring=true
    bLinearSpring=true
    bOverrideAngularDamping=true
    bOverrideLinearDamping=true
    bUseGravityOverride=true
    linearDampingOverride=0.9999
    linearSpringConstant=50
    simulationSpace=RootRelative
    alphaBoolBlend {blendOption=Linear}
    boundBone {boneName=Dyn_shorts_fr_R}
    componentAppliedLinearAccClamp=<5000 5000 5000>
    componentLinearAccScale=<0.5 0.5 0.5>
    componentLinearVelScale=<0.5 0.5 0.5>
    gravityOverride {z=-980}
    physicsBodyDefinitions [{
      boundBone {boneName=Dyn_shorts_fr_R}
      boxExtents=<10 10 10>
      constraintSetup {angularLimitsMax=<10 0 10>, angularLimitsMin=<-10 -7 -10>, angularTarget=<1 0 0>, linearAxesMax=<0 1 1>, linearAxesMin=<0 -1 -1>}
      localJointOffset=<0.7 0 0>
     }]
   }
   n12 = Anim Dynamics - Bone: Dyn_shorts_out_R(ComponentPose=@n11, Alpha=1, ExternalForce="0,0,0") {
    angularDampingOverride=0.9999
    angularSpringConstant=500
    bAngularSpring=true
    bLinearSpring=true
    bOverrideAngularDamping=true
    bOverrideLinearDamping=true
    bUseGravityOverride=true
    linearDampingOverride=0.9999
    linearSpringConstant=50
    simulationSpace=RootRelative
    alphaBoolBlend {blendOption=Linear}
    boundBone {boneName=Dyn_shorts_out_R}
    componentAppliedLinearAccClamp=<5000 5000 5000>
    componentLinearAccScale=<0.5 0.5 0.5>
    componentLinearVelScale=<0.5 0.5 0.5>
    gravityOverride {z=-980}
    physicsBodyDefinitions [{
      boundBone {boneName=Dyn_shorts_out_R}
      boxExtents=<10 10 10>
      constraintSetup {angularLimitsMax=<10 4 10>, angularLimitsMin=<-10 -7 -10>, angularTarget=<1 0 0>, linearAxesMax=<0 1 1>, linearAxesMin=<0 -1 -1>}
      localJointOffset=<0.7 0 0>
     }]
   }
   n13 = Anim Dynamics - Bone: Dyn_shorts_bk_R(ComponentPose=@n12, Alpha=1, ExternalForce="0,0,0") {
    angularDampingOverride=0.9999
    angularSpringConstant=500
    bAngularSpring=true
    bLinearSpring=true
    bOverrideAngularDamping=true
    bOverrideLinearDamping=true
    bUseGravityOverride=true
    linearDampingOverride=0.9999
    linearSpringConstant=50
    simulationSpace=RootRelative
    alphaBoolBlend {blendOption=Linear}
    boundBone {boneName=Dyn_shorts_bk_R}
    componentAppliedLinearAccClamp=<5000 5000 5000>
    componentLinearAccScale=<0.5 0.5 0.5>
    componentLinearVelScale=<0.5 0.5 0.5>
    gravityOverride {z=-980}
    physicsBodyDefinitions [{
      boundBone {boneName=Dyn_shorts_bk_R}
      boxExtents=<10 10 10>
      constraintSetup {angularLimitsMax=<10 4 10>, angularLimitsMin=<-10 -7 -10>, angularTarget=<1 0 0>, linearAxesMax=<0 1 1>, linearAxesMin=<0 -1 -1>}
      localJointOffset=<0.7 0 0>
     }]
   }
   n14 = Anim Dynamics - Bone: Dyn_sleeves_Down_L(ComponentPose=@n13, Alpha=1, ExternalForce="0,0,0") {
    angularDampingOverride=0.9999
    angularSpringConstant=500
    bAngularSpring=true
    bLinearSpring=true
    bOverrideAngularDamping=true
    bOverrideLinearDamping=true
    bUseGravityOverride=true
    linearDampingOverride=0.9999
    linearSpringConstant=20
    simulationSpace=RootRelative
    alphaBoolBlend {blendOption=Linear}
    boundBone {boneName=Dyn_sleeves_Down_L}
    componentAppliedLinearAccClamp=<5000 5000 5000>
    componentLinearAccScale=<0.5 0.5 0.5>
    componentLinearVelScale=<0.5 0.5 0.5>
    gravityOverride {z=-980}
    physicsBodyDefinitions [{
      boundBone {boneName=Dyn_sleeves_Down_L}
      boxExtents=<10 10 10>
      constraintSetup {angularLimitsMax=<10 10 10>, angularLimitsMin=<-10 -10 -10>, angularTarget=<1 0 0>, linearAxesMax=<0 1 1>, linearAxesMin=<0 -1 -1>}
      localJointOffset=<0.7 0 0>
     }]
   }
   n15 = Anim Dynamics - Bone: Dyn_sleeves_up_L(ComponentPose=@n14, Alpha=1, ExternalForce="0,0,0") {
    angularDampingOverride=0.9999
    angularSpringConstant=500
    bAngularSpring=true
    bLinearSpring=true
    bOverrideAngularDamping=true
    bOverrideLinearDamping=true
    bUseGravityOverride=true
    linearDampingOverride=0.9999
    linearSpringConstant=20
    simulationSpace=RootRelative
    alphaBoolBlend {blendOption=Linear}
    boundBone {boneName=Dyn_sleeves_up_L}
    componentAppliedLinearAccClamp=<5000 5000 5000>
    componentLinearAccScale=<0.5 0.5 0.5>
    componentLinearVelScale=<0.5 0.5 0.5>
    gravityOverride {z=-980}
    physicsBodyDefinitions [{
      boundBone {boneName=Dyn_sleeves_up_L}
      boxExtents=<10 10 10>
      constraintSetup {angularLimitsMax=<10 10 10>, angularLimitsMin=<-10 -10 -10>, angularTarget=<1 0 0>, linearAxesMax=<0 1 1>, linearAxesMin=<0 -1 -1>}
      localJointOffset=<0.7 0 0>
     }]
   }
   n16 = Anim Dynamics - Bone: Dyn_sleeves_Down_R(ComponentPose=@n15, Alpha=1, ExternalForce="0,0,0") {
    angularDampingOverride=0.9999
    angularSpringConstant=500
    bAngularSpring=true
    bLinearSpring=true
    bOverrideAngularDamping=true
    bOverrideLinearDamping=true
    bUseGravityOverride=true
    linearDampingOverride=0.9999
    linearSpringConstant=20
    simulationSpace=RootRelative
    alphaBoolBlend {blendOption=Linear}
    boundBone {boneName=Dyn_sleeves_Down_R}
    componentAppliedLinearAccClamp=<5000 5000 5000>
    componentLinearAccScale=<0.5 0.5 0.5>
    componentLinearVelScale=<0.5 0.5 0.5>
    gravityOverride {z=-980}
    physicsBodyDefinitions [{
      boundBone {boneName=Dyn_sleeves_Down_R}
      boxExtents=<10 10 10>
      constraintSetup {angularLimitsMax=<10 10 10>, angularLimitsMin=<-10 -10 -10>, angularTarget=<1 0 0>, linearAxesMax=<0 1 1>, linearAxesMin=<0 -1 -1>}
      localJointOffset=<0.7 0 0>
     }]
   }
   n17 = Anim Dynamics - Bone: Dyn_sleeves_up_R(ComponentPose=@n16, Alpha=1, ExternalForce="0,0,0") {
    angularDampingOverride=0.9999
    angularSpringConstant=500
    bAngularSpring=true
    bLinearSpring=true
    bOverrideAngularDamping=true
    bOverrideLinearDamping=true
    bUseGravityOverride=true
    linearDampingOverride=0.9999
    linearSpringConstant=20
    simulationSpace=RootRelative
    alphaBoolBlend {blendOption=Linear}
    boundBone {boneName=Dyn_sleeves_up_R}
    componentAppliedLinearAccClamp=<5000 5000 5000>
    componentLinearAccScale=<0.5 0.5 0.5>
    componentLinearVelScale=<0.5 0.5 0.5>
    gravityOverride {z=-980}
    physicsBodyDefinitions [{
      boundBone {boneName=Dyn_sleeves_up_R}
      boxExtents=<10 10 10>
      constraintSetup {angularLimitsMax=<10 10 10>, angularLimitsMin=<-10 -10 -10>, angularTarget=<1 0 0>, linearAxesMax=<0 1 1>, linearAxesMin=<0 -1 -1>}
      localJointOffset=<0.7 0 0>
     }]
   }
   n18 = Anim Dynamics - Bone: Dyn_amulet_01(ComponentPose=@n17, Alpha=1, ExternalForce="0,0,0") {
    angularDampingOverride=0.9999
    angularSpringConstant=500
    bAngularSpring=true
    bOverrideAngularDamping=true
    bUseGravityOverride=true
    bUseSphericalLimits=true
    linearDampingOverride=0.9999
    linearSpringConstant=500
    simulationSpace=RootRelative
    alphaBoolBlend {blendOption=Linear}
    boundBone {boneName=Dyn_amulet_01}
    componentAppliedLinearAccClamp=<5000 5000 5000>
    componentLinearAccScale=<0.5 0.5 0.5>
    componentLinearVelScale=<0.5 0.5 0.5>
    gravityOverride {z=-980}
    physicsBodyDefinitions [{
      sphereCollisionRadius=5
      boundBone {boneName=Dyn_amulet_01}
      boxExtents=<10 10 10>
      constraintSetup {angularLimitsMax=<0 0 5>, angularLimitsMin=<0 -30 -5>, angularTarget=<1 0 0>}
      localJointOffset=<1 0 0>
     }]
   }
   n19 = Anim Dynamics - Bone: Dyn_amulet_02(ComponentPose=@n18, Alpha=1, ExternalForce="0,0,0") {
    angularDampingOverride=0.9999
    angularSpringConstant=500
    bAngularSpring=true
    bOverrideAngularDamping=true
    bUseGravityOverride=true
    linearDampingOverride=0.9999
    linearSpringConstant=500
    simulationSpace=RootRelative
    alphaBoolBlend {blendOption=Linear}
    boundBone {boneName=Dyn_amulet_02}
    componentAppliedLinearAccClamp=<5000 5000 5000>
    componentLinearAccScale=<0.5 0.5 0.5>
    componentLinearVelScale=<0.5 0.5 0.5>
    gravityOverride {z=-980}
    physicsBodyDefinitions [{
      boundBone {boneName=Dyn_amulet_02}
      boxExtents=<10 10 10>
      constraintSetup {angularLimitsMax=<0 0 5>, angularLimitsMin=<0 -20 -5>, angularTarget=<1 0 0>}
      localJointOffset=<1 0 0>
     }]
   }
   n20 = Anim Dynamics - Bone: Dyn_amulet_03(ComponentPose=@n19, Alpha=1, ExternalForce="0,0,0") {
    angularDampingOverride=0.9999
    angularSpringConstant=500
    bAngularSpring=true
    bOverrideAngularDamping=true
    bUseGravityOverride=true
    bUseSphericalLimits=true
    linearDampingOverride=0.9999
    linearSpringConstant=500
    simulationSpace=RootRelative
    alphaBoolBlend {blendOption=Linear}
    boundBone {boneName=Dyn_amulet_03}
    componentAppliedLinearAccClamp=<5000 5000 5000>
    componentLinearAccScale=<0.5 0.5 0.5>
    componentLinearVelScale=<0.5 0.5 0.5>
    gravityOverride {z=-980}
    physicsBodyDefinitions [{
      collisionType=CustomSphere
      sphereCollisionRadius=5
      boundBone {boneName=Dyn_amulet_03}
      boxExtents=<10 10 10>
      constraintSetup {angularLimitsMax=<0 0 5>, angularLimitsMin=<0 -15 -5>, angularTarget=<1 0 0>}
      localJointOffset=<1 0 0>
     }]
    sphericalLimits [{limitRadius=7.5, drivingBone {boneName=clavicle_r}, sphereLocalOffset=<-8 0 0>}]
   }
   n21 = Anim Dynamics - Bone: Dyn_Laces_L(ComponentPose=@n20, Alpha=1, ExternalForce="0,0,0") {
    angularDampingOverride=0.9999
    angularSpringConstant=500
    bAngularSpring=true
    bOverrideAngularDamping=true
    bUseGravityOverride=true
    linearDampingOverride=0.9999
    linearSpringConstant=500
    simulationSpace=RootRelative
    alphaBoolBlend {blendOption=Linear}
    boundBone {boneName=Dyn_Laces_L}
    componentAppliedLinearAccClamp=<5000 5000 5000>
    componentLinearAccScale=<0.5 0.5 0.5>
    componentLinearVelScale=<0.5 0.5 0.5>
    gravityOverride {z=-980}
    physicsBodyDefinitions [{
      boundBone {boneName=Dyn_Laces_L}
      boxExtents=<10 10 10>
      constraintSetup {angularLimitsMax=<0 0 10>, angularLimitsMin=<0 -20 -10>, angularTarget=<1 0 0>}
      localJointOffset=<1 0 0>
     }]
   }
   n22 = Anim Dynamics - Bone: Dyn_Laces_R(ComponentPose=@n21, Alpha=1, ExternalForce="0,0,0") {
    angularDampingOverride=0.9999
    angularSpringConstant=500
    bAngularSpring=true
    bOverrideAngularDamping=true
    bUseGravityOverride=true
    linearDampingOverride=0.9999
    linearSpringConstant=500
    simulationSpace=RootRelative
    alphaBoolBlend {blendOption=Linear}
    boundBone {boneName=Dyn_Laces_R}
    componentAppliedLinearAccClamp=<5000 5000 5000>
    componentLinearAccScale=<0.5 0.5 0.5>
    componentLinearVelScale=<0.5 0.5 0.5>
    gravityOverride {z=-980}
    physicsBodyDefinitions [{
      boundBone {boneName=Dyn_Laces_R}
      boxExtents=<10 10 10>
      constraintSetup {angularLimitsMax=<0 0 10>, angularLimitsMin=<0 -20 -10>, angularTarget=<1 0 0>}
      localJointOffset=<1 0 0>
     }]
   }
   n23 = Anim Dynamics - Bone: Dyn_bracer_L(ComponentPose=@n22, Alpha=1, ExternalForce="0,0,0") {
    angularDampingOverride=0.9999
    angularSpringConstant=100
    bLinearSpring=true
    bOverrideAngularDamping=true
    bUseGravityOverride=true
    linearDampingOverride=0.9999
    linearSpringConstant=500
    simulationSpace=BoneRelative
    alphaBoolBlend {blendOption=Linear}
    boundBone {boneName=Dyn_bracer_L}
    componentAppliedLinearAccClamp=<1500 1500 1500>
    componentLinearAccScale=<1 1 1>
    componentLinearVelScale=<1 1 1>
    gravityOverride {z=-980}
    physicsBodyDefinitions [{
      boundBone {boneName=Dyn_bracer_L}
      boxExtents=<20 20 20>
      constraintSetup {angularLimitsMax=<5 10 10>, angularLimitsMin=<-5 0 0>, angularTarget=<1 0 0>}
      localJointOffset=<0.6 0 0>
     }]
    relativeSpaceBone {boneName=hand_l}
   }
   n24 = Anim Dynamics - Bone: Dyn_bracer_L_down(ComponentPose=@n23, Alpha=1, ExternalForce="0,0,0") {
    angularSpringConstant=500
    bAngularSpring=true
    bOverrideLinearDamping=true
    linearDampingOverride=0.9999
    simulationSpace=BoneRelative
    alphaBoolBlend {blendOption=Linear}
    boundBone {boneName=Dyn_bracer_L_down}
    componentAppliedLinearAccClamp=<1500 1500 1500>
    componentLinearAccScale=<1 1 1>
    componentLinearVelScale=<1 1 1>
    physicsBodyDefinitions [{
      boundBone {boneName=Dyn_bracer_L_down}
      boxExtents=<20 20 20>
      constraintSetup {angularLimitsMax=<0 0 0>, angularLimitsMin=<0 0 0>, angularTarget=<1 0 0>, linearAxesMax=<0 0 1>, linearAxesMin=<0 -1 -1>}
      localJointOffset=<-0.6 0 0>
     }]
    relativeSpaceBone {boneName=hand_l}
   }
   n25 = Anim Dynamics - Bone: Dyn_bracer_L_up(ComponentPose=@n24, Alpha=1, ExternalForce="0,0,0") {
    angularSpringConstant=500
    bAngularSpring=true
    bOverrideLinearDamping=true
    linearDampingOverride=0.9999
    simulationSpace=BoneRelative
    alphaBoolBlend {blendOption=Linear}
    boundBone {boneName=Dyn_bracer_L_up}
    componentAppliedLinearAccClamp=<1500 1500 1500>
    componentLinearAccScale=<1 1 1>
    componentLinearVelScale=<1 1 1>
    physicsBodyDefinitions [{
      boundBone {boneName=Dyn_bracer_L_up}
      boxExtents=<20 20 20>
      constraintSetup {angularLimitsMax=<0 0 0>, angularLimitsMin=<0 0 0>, angularTarget=<1 0 0>, linearAxesMax=<0 1 0>, linearAxesMin=<0 0 -1>}
      localJointOffset=<0.6 0 0>
     }]
    relativeSpaceBone {boneName=hand_l}
   }
   Output Pose(Result=Component To Local(ComponentPose=@n25)) {layerGroup=None}
  ]}
 EventGraph {nodes [
   Event BlueprintUpdateAnimation()
   > Set NotEmote(NotEmote:bool=NOT Boolean(A=Get bIsPlayingEmote())) -> (Output_Get:bool)
   > Set 0 and 1 lod(0 and 1 lod:bool="integer <= integer"(A=Get CurrentLOD(), B=1)) -> (Output_Get:bool)
   > n1 = Cast To FortPlayerAnimInstance(Object=GetAnimInstance(self=GetOwnerSkeletalMeshComponent()))
   > Switch on EFortWeaponCoreAnimation!(execute, Selection:EFortWeaponCoreAnimation Enum=Get WeaponCoreAnim(self=@n1)) -> (
    Melee:Exec
    Pistol:Exec
    Shotgun:Exec
    PaperBlueprint:Exec
    Rifle:Exec
    MeleeOneHand:Exec
    MachinePistol:Exec
    RocketLauncher:Exec
    GrenadeLauncher:Exec
    GoingCommando:Exec
    AssaultRifle:Exec
    TacticalShotgun:Exec
    SniperRifle:Exec
    TrapPlacement:Exec
    ShoulderLauncher:Exec
    AbilityDecoTool:Exec
    Crossbow:Exec
    C4:Exec
    RemoteControl:Exec
    DualWield:Exec
    AR_BullPup:Exec
    AR_ForwardGrip:Exec
    MedPackPaddles:Exec
    SMG_P90:Exec
    AR_DrumGun:Exec
    Consumable_Small:Exec
    Consumable_Large:Exec
    Balloon:Exec
    MountedTurret:Exec
    CreativeTool:Exec
    ExplosiveBow:Exec
    AshtonIndigo:Exec
    AshtonChicago:Exec
    MeleeDualWield:Exec
    Unarmed:Exec
   )
   Event OnCharacterCustomizationLoaded()
   > n2 = Cast To PlayerPawn_Athena(Object=TryGetPawnOwner())
   > Branch(Condition=CheckForTail!(self:Custom Character Part Anim Instance Object Reference) -> (bHasTail:bool))
   > n3 = Branch(Condition=CheckForBackbling().bIsBackblingLong)
   n4 = GetSkeletalMeshForPartType(self=@n2, PartType=MiscOrTail)
   > SetVisibility!(execute, self:Scene Component Object Reference=@n4, bNewVisibility:bool=false, bPropagateToChildren:bool=false) -> (then)
   SetVisibility(execute=@n3.else, self=@n4, bNewVisibility=true, bPropagateToChildren=false)
  ]}
]
variables [
 0 and 1 lod {type=boolean}
 isSMG {type=boolean}
 NotEmote {type=boolean}
]
