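# Physics asset for the M_MED_PointyTemper male medium player cosmetic body, named as the collider set for its cape.
# One capsule per bone approximates the pelvis, upper spine, arms, hands, legs and feet, with mirrored left and right limbs and no head, neck or clavicle shapes.
# Constraints link the bodies in a hierarchy from the pelvis, all sharing the same swing and twist limit angles.
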PhysicsAsset /BRCosmetics/Characters/Player/Male/Medium/Bodies/M_MED_PointyTemper/Meshes/M_MED_PointyTemper_CapeColliders
bodies {
 calf_l [{shapeType=Capsule, length=44.9037, radius=12.2477, center=<-23.1582 -2.2427 0>, rotation=<rot 90 0 0>}]
 calf_r [{shapeType=Capsule, length=44.9037, radius=12.2477, center=<23.1582 2.2427 0>, rotation=<rot 90 0 180>}]
 foot_l [{shapeType=Capsule, length=16.0543, radius=10.1004, center=<-4.1955 6.0815 0>, rotation=<rot 0 14.9998 90>}]
 foot_r [{shapeType=Capsule, length=16.0543, radius=10.1004, center=<4.1955 -6.0815 0>, rotation=<rot 0 14.9998 -90>}]
 hand_l [{shapeType=Capsule, length=9.5015, radius=6.7933, center=<6.2703 0 0>, rotation=<rot 90 0 0>}]
 hand_r [{shapeType=Capsule, length=9.5015, radius=6.7933, center=<-6.2703 0 0>, rotation=<rot 90 0 180>}]
 lowerarm_l [{shapeType=Capsule, length=14.3387, radius=7.5574, center=<8.9852 0 0.2267>, rotation=<rot 90 0 0>}]
 lowerarm_r [{shapeType=Capsule, length=14.3387, radius=7.5574, center=<-8.9852 0 -0.2267>, rotation=<rot 90 0 180>}]
 pelvis [{shapeType=Capsule, length=8.8436, radius=16.7074}]
 spine_02 [{shapeType=Capsule, length=7.1866, radius=15.773}]
 spine_03 [{shapeType=Capsule, length=6.2861, radius=15.1536}]
 spine_04 [{shapeType=Capsule, length=5.1583, radius=14.7453}]
 spine_05 [{shapeType=Capsule, length=17.1959, radius=12.7259}]
 thigh_l [{shapeType=Capsule, length=17.5494, radius=12.5057, center=<-18.4217 -1.038 0>, rotation=<rot 90 180 2.664>}]
 thigh_r [{shapeType=Capsule, length=17.5494, radius=12.5057, center=<18.4217 1.038 0>, rotation=<rot 90 179.7762 -177.5598>}]
 upperarm_l [{shapeType=Capsule, length=17.1712, radius=7.7501, center=<10.0455 0 0>, rotation=<rot 90 0 0>}]
 upperarm_r [{shapeType=Capsule, length=17.1712, radius=7.7501, center=<-10.0455 0 0>, rotation=<rot 90 0 180>}]
}
constraints [
 {bone1Name=spine_02, bone2Name=pelvis, swing1LimitDegrees=45, swing2LimitDegrees=45, twistLimitDegrees=45}
 {bone1Name=spine_03, bone2Name=spine_02, swing1LimitDegrees=45, swing2LimitDegrees=45, twistLimitDegrees=45}
 {bone1Name=spine_04, bone2Name=spine_03, swing1LimitDegrees=45, swing2LimitDegrees=45, twistLimitDegrees=45}
 {bone1Name=spine_05, bone2Name=spine_04, swing1LimitDegrees=45, swing2LimitDegrees=45, twistLimitDegrees=45}
 {bone1Name=upperarm_l, bone2Name=spine_05, swing1LimitDegrees=45, swing2LimitDegrees=45, twistLimitDegrees=45}
 {bone1Name=lowerarm_l, bone2Name=upperarm_l, swing1LimitDegrees=45, swing2LimitDegrees=45, twistLimitDegrees=45}
 {bone1Name=hand_l, bone2Name=lowerarm_l, swing1LimitDegrees=45, swing2LimitDegrees=45, twistLimitDegrees=45}
 {bone1Name=foot_l, bone2Name=calf_l, swing1LimitDegrees=45, swing2LimitDegrees=45, twistLimitDegrees=45}
 {bone1Name=thigh_l, bone2Name=pelvis, swing1LimitDegrees=45, swing2LimitDegrees=45, twistLimitDegrees=45}
 {bone1Name=hand_r, bone2Name=lowerarm_r, swing1LimitDegrees=45, swing2LimitDegrees=45, twistLimitDegrees=45}
 {bone1Name=foot_r, bone2Name=calf_r, swing1LimitDegrees=45, swing2LimitDegrees=45, twistLimitDegrees=45}
 {bone1Name=calf_r, bone2Name=thigh_r, swing1LimitDegrees=45, swing2LimitDegrees=45, twistLimitDegrees=45}
 {bone1Name=thigh_r, bone2Name=pelvis, swing1LimitDegrees=45, swing2LimitDegrees=45, twistLimitDegrees=45}
 {bone1Name=upperarm_r, bone2Name=spine_05, swing1LimitDegrees=45, swing2LimitDegrees=45, twistLimitDegrees=45}
 {bone1Name=lowerarm_r, bone2Name=upperarm_r, swing1LimitDegrees=45, swing2LimitDegrees=45, twistLimitDegrees=45}
 {bone1Name=calf_l, bone2Name=thigh_l, swing1LimitDegrees=45, swing2LimitDegrees=45, twistLimitDegrees=45}
]
